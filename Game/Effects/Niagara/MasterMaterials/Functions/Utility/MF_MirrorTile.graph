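# MF_MirrorTile is a library material function in the Niagara effects utilities that mirrors and tiles a greyscale input such as a UV channel.
# It doubles the scalar Input and multiplies it by Tile, which defaults to 1 for a single reflection.
# The floor of that value modulo 2 selects between its fractional part and the inverse through a lerp, producing a branch-free triangle wave.
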MaterialFunction /Game/Effects/Niagara/MasterMaterials/Functions/Utility/MF_MirrorTile
graph {
 properties {bExposeToLibrary=true, description=Mirrors and tiles a greyscale input such as UV channels}
 nodes [
  n1 = FunctionInput(Preview=ComponentMask(None=TextureCoordinate()) {r=true}) {inputName=Input, inputType=FunctionInput_Scalar}
  n2 = FunctionInput(Preview=Constant() {r=1}) {
   bUsePreviewValueAsDefault=true
   description=Defaults to 1 for a single reflection
   inputName=Tile
   inputType=FunctionInput_Scalar
   sortPriority=1
  }
  n3 = Multiply(A=Add(A=@n1, B=@n1), B=@n2)
  n4 = Frac(None=@n3)
  n5 = Modulo!(A=@n3, B=Constant() {r=2})
  FunctionOutput(None=LinearInterpolate(A=@n4, B=OneMinus(None=@n4), Alpha=Floor(None=@n5)))
 ]
}
