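# Translucent, two-sided unlit Niagara material for the heat shield effect of the SpacePlunge skydiving contrail.
# Distorted, scrolling noise in switchable polar or tiled UVs passes a smoothstep cutout into a color gradient tinted by particle color.
# Opacity combines gradient and particle alpha with reroute-fed masking (Fresnel and sphere/linear masks are declared), adding depth fade only on higher-end platforms.
# A masked world position offset pulses along normals using sine time.
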Material /BRCosmetics/Athena/Effects/Particles/Skydiving/NS_Contrail_SpacePlunge/Materials/M_Contrail_SpacePlunge_HeatShield
graph {
 properties {
  bAutomaticallySetUsageInEditor=false
  bHasPixelAnimation=true
  bUsedWithNiagaraMeshParticles=true
  bUsedWithNiagaraRibbons=true
  bUsedWithNiagaraSprites=true
  blendMode=BLEND_Translucent
  shadingModel=MSM_Unlit
  twoSided=true
  usedShadingModels=Unlit
 }
 nodes [
  n1 = ScalarParameter() {defaultValue=0.75, group=Cutout, parameterName=Cutout Value, sortPriority=33}
  n2 = ScalarParameter() {defaultValue=0.3, group=Cutout, parameterName=Cutout Spread}
  n3 = Subtract(A=@n1, B=@n2)
  n4 = Add(A=@n2, B=@n1)
  n5 = DynamicParameter() {defaultValue=<rgba 0 0 0 0>, paramNames [Noise Offset U, Noise Offset V, Radial Noise Offset U, Radial Noise Offset V]}
  n6 = Add(A=AppendVector(A=@n5.Noise Offset U, B=@n5.Noise Offset U), B=TextureCoordinate())
  n7 = MaterialFunctionCall!(
   UV Coord=@n6
   LenghtScale=ScalarParameter() {defaultValue=1.5, group=UVs, parameterName=Noise Polar Length Scale, sortPriority=33}
   RadialScale=ScalarParameter() {defaultValue=0.125, group=UVs, parameterName=Noise Polar Radial Scale, sortPriority=34}
  ) {materialFunction=/Game/Global/MaterialFunctions/TechArt/MF_PolarCoordinates}
  n8 = Multiply(
   A=AppendVector(
    A=ScalarParameter() {defaultValue=0.1, group=UVs, parameterName=Noise Tile U, sortPriority=35}
    B=ScalarParameter() {defaultValue=0.5, group=UVs, parameterName=Noise Tile V, sortPriority=36}
   )
   B=@n6
  )
  n9 = ScalarParameter() {defaultValue=-2, group=UVs, parameterName=Noise Scroll Speed U, sortPriority=37}
  n10 = ScalarParameter() {group=UVs, parameterName=Noise Scroll Speed V, sortPriority=38}
  n11 = MaterialFunctionCall(
   UV Coord=@n6
   LenghtScale=ScalarParameter() {defaultValue=1, group=UVs - Distortion, parameterName=UV Distortion Polar Length Scale, sortPriority=34}
   RadialScale=ScalarParameter() {defaultValue=1, group=UVs - Distortion, parameterName=UV Distortion Polar Radial Scale, sortPriority=35}
  ) {materialFunction=&MF_PolarCoordinates}
  n12 = Multiply(
   A=AppendVector(
    A=ScalarParameter() {defaultValue=1, group=UVs - Distortion, parameterName=UV Distortion Tile U, sortPriority=36}
    B=ScalarParameter() {defaultValue=1, group=UVs - Distortion, parameterName=UV Distortion Tile V, sortPriority=37}
   )
   B=@n6
  )
  n13 = ScalarParameter() {defaultValue=-0.5, group=UVs - Distortion, parameterName=UV Distortion Scroll Speed U, sortPriority=38}
  n14 = ScalarParameter() {defaultValue=0.1, group=UVs - Distortion, parameterName=UV DIstortion Scroll Speed V, sortPriority=39}
  n15 = DynamicParameter() {defaultValue=<rgba 0 0 0 0>, paramNames [Noise Offset U, Noise Offset V, Radial Noise Offset U, Radial Noise Offset V]}
  n16 = Add(
   A=Panner(
    Coordinate=StaticSwitchParameter(True=@n11.Result, False=@n12) {defaultValue=true, group=UVs, parameterName=Use Polar Coords}
    Speed=AppendVector(A=@n13, B=@n14)
   ) {speedX=-1}
   B=AppendVector(A=@n15.Radial Noise Offset U, B=@n15.Radial Noise Offset U)
  )
  n17 = Constant()
  n18 = TextureObjectParameter() {group=Textures, parameterName=Noise Texture, texture=/Game/Effects/Fort_Effects/Textures/Tile/T_GaussianNoise_01_M}
  n19 = Multiply(
   A=StaticSwitchParameter(
    True=TextureSample!(UVs=@n16, Tex=@n18, "DDX(UVs)"=@n17, "DDY(UVs)"=@n17, Apply View MipBias) {mipValueMode=TMVM_Derivative, texture=&T_GaussianNoise_01_M}.R
    False=TextureSample!(UVs=@n16, Tex=@n18, Apply View MipBias) {texture=&T_GaussianNoise_01_M}.R
   ) {defaultValue=true, group=UVs, parameterName=Use Polar Coords}
   B=OneMinus(None=NamedRerouteUsage())
  )
  n20 = Add(
   A=Panner(
    Coordinate=StaticSwitchParameter(True=@n7.Result, False=@n8) {defaultValue=true, group=UVs, parameterName=Use Polar Coords}
    Speed=AppendVector(A=@n9, B=@n10)
   ) {speedX=-0.1}
   B=Multiply(
    A=ScalarParameter() {defaultValue=1, group=UVs - Distortion, parameterName=UV Distortion Str}
    B=NamedRerouteDeclaration(None=@n19) {name=UV Distortion, nodeColor=<rgba 1 0 0.2344 1>}
   ) {constB=0.2}
  )
  n21 = TextureObjectParameter() {group=Textures, parameterName=Noise Texture, texture=&T_GaussianNoise_01_M}
  n22 = Constant()
  n23 = StaticSwitchParameter(
   True=TextureSample!(UVs=@n20, Tex=@n21, "DDX(UVs)"=@n22, "DDY(UVs)"=@n22, Apply View MipBias) {mipValueMode=TMVM_Derivative, texture=&T_GaussianNoise_01_M}.R
   False=TextureSample!(UVs=@n20, Tex=@n21, Apply View MipBias) {texture=&T_GaussianNoise_01_M}.R
  ) {defaultValue=true, group=UVs, parameterName=Use Polar Coords}
  n24 = Power(Base=NamedRerouteUsage()) {constExponent=1}
  n25 = TextureObjectParameter() {
   desc="Color Remap\r\nNote: you can override this in Niagara by binding a color curve in the system node"
   group=Textures
   parameterName=Color Gradient
   sortPriority=33
   texture=/Game/Effects/Fort_Effects/Textures/Grad/T_ColorGrad_FireRed
  }
  n26 = TextureSample(
   UVs=SmoothStep(Min=Saturate(None=@n3), Max=Saturate(None=@n4), Value=LinearInterpolate(A=@n23, Alpha=@n24)) {constMax=0.8, constMin=0.5}
   Tex=@n25
  )
  n27 = ParticleColor()
  n28 = Multiply(A=@n26.RGB, B=@n27.RGB)
  n29 = VertexColor()
  MaterialFunctionCall(Emissive=LinearInterpolate(A=@n28, B=Multiply(A=@n28, B=@n29), Alpha=ScalarParameter() {group=Vert Color, parameterName=Vert Color Contribution})) {materialFunction=&MF_FX_EmissiveAdjustments}
  n30 = DotProduct(A=CameraVectorWS(), B=Multiply(A=VertexNormalWS(), B=TwoSidedSign()))
  n31 = ScalarParameter() {defaultValue=0.3, group=Fresnal, parameterName=Fresnal Contrast}
  n32 = Saturate(None=Power(
    Base=MaterialFunctionCall(In=@n30, Contrast=@n31) {materialFunction=&CheapContrast}.Result
    Exp=ScalarParameter() {defaultValue=2, group=Fresnal, parameterName=Fresnel Exp, sortPriority=33}
   ))
  NamedRerouteDeclaration(None=LinearInterpolate(A=@n32, B=OneMinus(None=@n32), Alpha=ScalarParameter() {group=Fresnal, parameterName=Frenal Inversion, sortPriority=34})) {name=Fresnal, nodeColor=<rgba 0.0078 0 1 1>}
  n33 = ScalarParameter() {defaultValue=1, group=Mask, parameterName=Mask Tile, sortPriority=35}
  n34 = TextureCoordinate()
  n35 = ComponentMask(None=@n34) {r=true}
  n36 = Multiply(A=@n33, B=@n35)
  n37 = MaterialFunctionCall!(Input=@n35, Tile=@n33) {materialFunction=/Game/Effects/Niagara/MasterMaterials/Functions/Utility/MF_MirrorTile}
  n38 = ScalarParameter() {group=Mask, parameterName=Reflect Linear Gradient, sortPriority=34}
  n39 = ScalarParameter() {defaultValue=1, group=Mask, parameterName=Mask Tile, sortPriority=35}
  n40 = ComponentMask(None=@n34) {g=true}
  n41 = Multiply(A=@n39, B=@n40)
  n42 = MaterialFunctionCall(Input=@n40, Tile=@n39) {materialFunction=&MF_MirrorTile}
  n43 = ScalarParameter() {group=Mask, parameterName=Reflect Linear Gradient, sortPriority=34}
  n44 = LinearInterpolate(
   A=LinearInterpolate(
    A=LinearInterpolate(A=@n36, B=@n37.Result, Alpha=@n38)
    B=LinearInterpolate(A=@n41, B=@n42.Result, Alpha=@n43)
    Alpha=ScalarParameter() {group=Mask, parameterName=Vertical Gradient, sortPriority=33}
   )
   B=SphereMask(
    A=TextureCoordinate()
    B=Constant() {r=0.5}
    Radius=ScalarParameter() {defaultValue=0.3, group=Mask, parameterName=Sphere Mask Radius, sortPriority=36}
    Hardness=ScalarParameter() {defaultValue=0.2, group=Mask, parameterName=Sphere Mask Hardness, sortPriority=37}
   ) {attenuationRadius=0.5, hardnessPercent=0}
   Alpha=ScalarParameter() {defaultValue=1, desc="0 uses linear gradient, 1 uses sphere mask", group=Mask, parameterName=UseSphereMask}
  )
  n45 = OneMinus(None=@n44)
  n46 = ScalarParameter() {defaultValue=1, group=WPO, parameterName=WPO Mask Inversion, sortPriority=34}
  NamedRerouteDeclaration(None=Power(
    Base=LinearInterpolate(A=@n44, B=@n45, Alpha=@n46) {constAlpha=1}
    Exp=ScalarParameter() {defaultValue=2, group=WPO, parameterName=WPO Mask Power, sortPriority=33}
   )) {name=WPO Mask, nodeColor=<rgba 0.3125 1 0 1>}
  n47 = LinearInterpolate(A=@n44, B=OneMinus(None=@n44), Alpha=ScalarParameter() {group=Mask, parameterName=Mask Inversion, sortPriority=35}) {constAlpha=1}
  n48 = ScalarParameter() {defaultValue=1, group=Mask, parameterName=Mask Power, sortPriority=31}
  NamedRerouteDeclaration(None=Saturate(None=Power(Base=@n47, Exp=@n48))) {name=Mask, nodeColor=<rgba 0 0.9531 1 1>}
  n49 = Sine(None=Time())
  n50 = ScalarParameter() {defaultValue=1, group=WPO, parameterName=WPO Sin Strength, sortPriority=37}
  n51 = NamedRerouteUsage()
  n52 = ScalarParameter() {group=WPO, parameterName=WPO UV Distortion Strength, sortPriority=36}
  n53 = Multiply(
   A=Multiply(A=VertexNormalWS(), B=ScalarParameter() {defaultValue=15, group=WPO, parameterName=WPO Vert Normal, sortPriority=35})
   B=Add(A=Multiply(A=@n49, B=@n50), B=Multiply(A=@n51, B=@n52))
  )
  n54 = NamedRerouteUsage()
  NamedRerouteDeclaration(None=Multiply(A=Multiply(A=@n53, B=@n54), B=ScalarParameter() {defaultValue=1, group=WPO, parameterName=WPO Global}) {constB=100}) {name=WPO, nodeColor=<rgba 0.3594 1 0 1>}
  NamedRerouteUsage()
  n55 = Multiply(A=NamedRerouteUsage(), B=NamedRerouteUsage())
  n56 = DepthFade(Opacity=@n55, FadeDistance=ScalarParameter() {defaultValue=100, group=Depth Fade, parameterName=Depth Fade Distance})
  n57 = Multiply(
   A=Multiply(A=@n26.A, B=@n27.A)
   B=MaterialFunctionCall!("Xb1,PS4"=@n56, "PC-High&&Epic,XSX,PS5"=@n56, "PC-Perf,Android"=@n55, Switch=@n55, "PC-Low(NotPerf)&&Medium"=@n55) {materialFunction=/Game/Effects/Niagara/MasterMaterials/Functions/Common/PlatformSwitch_4D}.Result
  )
  n58 = VertexColor()
  Saturate(None=LinearInterpolate(A=@n57, B=Multiply(A=@n57, B=@n58.A), Alpha=ScalarParameter() {group=Vert Color, parameterName=Vert Alpha Contribution, sortPriority=33}))
 ]
}
