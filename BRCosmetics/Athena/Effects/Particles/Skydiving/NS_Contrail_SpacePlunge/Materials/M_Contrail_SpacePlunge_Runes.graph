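# Additive, unlit, two-sided Niagara material for the rune particles of the SpacePlunge skydiving contrail, usable on sprites, ribbons and meshes.
# A SubUV rune texture is scaled per channel, selected through a looping four-way lerp (red in three slots, alpha by default in the fourth) and optionally inverted.
# The mask indexes an overridable color gradient; dynamic parameters drive channel choice, gradient blend and emissive strength.
# Particle color tints the result, and particle alpha multiplies the mask.
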Material /BRCosmetics/Athena/Effects/Particles/Skydiving/NS_Contrail_SpacePlunge/Materials/M_Contrail_SpacePlunge_Runes
graph {
 properties {
  bAutomaticallySetUsageInEditor=false
  bUsedWithNiagaraMeshParticles=true
  bUsedWithNiagaraRibbons=true
  bUsedWithNiagaraSprites=true
  blendMode=BLEND_Additive
  shadingModel=MSM_Unlit
  twoSided=true
  usedShadingModels=Unlit
 }
 nodes [
  Constant()
  n1 = ParticleColor()
  n2 = TextureSampleParameterSubUV!(UVs, Apply View MipBias) {
   group=Mask
   parameterName=Mask Texture
   texture=/BRCosmetics/Athena/Effects/Particles/Skydiving/NS_Contrail_SpacePlunge/Textures/T_Contrail_SpacePlunge_Runes
  }
  n3 = VectorParameter() {
   desc=This acts as a mulitiplier per channel of the mask texture
   group=Mask
   parameterName=Mask Multiplier
   sortPriority=33
   defaultValue=<rgba 1 1 1 0>
  }
  n4 = MaterialFunctionCall!(Float4=Multiply(A=@n2.RGBA, B=@n3.RGBA)) {materialFunction=/Game/Global/MaterialFunctions/MF_BreakComponents}
  n5 = MaterialFunctionCall!(
   Mask 1=@n4.R
   Mask 2=@n4.R
   Mask 3=@n4.R
   Mask 4=ChannelMaskParameter(None=@n4.RGBA) {maskChannel=Alpha, parameterName=Alpha Channel Override, defaultValue=<rgba 0 0 0 1>}.R
   Blend=DynamicParameter() {paramNames [Mask Channel PIcker, Color Gradient Contribution, Emissive, Param4]}.Mask Channel PIcker
  ) {materialFunction=/Game/Effects/Niagara/MasterMaterials/Functions/Utility/MF_Lerp_4_Scalar_Looping}
  n6 = LinearInterpolate(
   A=@n5.Result
   B=OneMinus(None=@n5.Result)
   Alpha=ScalarParameter() {desc="Lerps mask to inverted (not clamped) 1 = inverted", group=Mask, parameterName=Mask Inversion, sortPriority=34}
  )
  n7 = TextureObjectParameter(UVs=@n6) {
   desc=This can be overridden by binding a color curve in niagara system
   group=Color Gradient
   parameterName=Color Gradient
   texture=/Game/Effects/Fort_Effects/Textures/Grad/T_ColorGrad_FireRed
  }
  n8 = DynamicParameter() {paramNames [Mask Channel PIcker, Color Gradient Contribution, Emissive, Param4]}
  n9 = LinearInterpolate(A=@n6, B=TextureSample(UVs=@n6, Tex=@n7).RGB, Alpha=@n8.Color Gradient Contribution)
  MaterialFunctionCall(Emissive=Multiply(A=Multiply(A=@n1.RGB, B=@n9), B=@n8.Emissive)) {materialFunction=&MF_FX_EmissiveAdjustments}
  Multiply(A=@n1.A, B=@n6)
 ]
}
