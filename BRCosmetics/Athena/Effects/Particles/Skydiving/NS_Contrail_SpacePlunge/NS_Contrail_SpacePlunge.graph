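# Niagara skydiving contrail "SpacePlunge", a fiery plunge effect with a heat shield mesh, bone-attached body glows, debris, sparks, animated rune sprites and lightning meshes.
# User.ParaGlideLeanAlpha intensifies it, driving color, spawn probability, upward gravity and shield placement.
# The shield also adjusts to vertical view angle and time of day, while BodyGlows uses partition index to pick each particle's bone and size.
# Front-end modifiers retune size, brightness and offsets outside gameplay.
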NiagaraSystem /BRCosmetics/Athena/Effects/Particles/Skydiving/NS_Contrail_SpacePlunge/NS_Contrail_SpacePlunge
dependencies {
 usedDataInterfaces [
  &NiagaraDataInterfaceVectorCurve
  &NiagaraDataInterfaceCamera
  &NiagaraDataInterfaceColorCurve
  &NiagaraDataInterfaceCurve
  &NiagaraDataInterfaceSpriteRendererInfo
  &NiagaraDataInterfaceArrayFloat2
  &NiagaraDataInterfaceSkeletalMesh
 ]
 usedDynamicInputs []
 usedModules [
  &self:SystemSpawnScript.SystemScriptSource.SystemScriptGraph.NiagaraNodeAssignment_0.SetVariables_#1
  &self:SystemSpawnScript.SystemScriptSource.SystemScriptGraph.NiagaraNodeAssignment_1.SetVariables_#2
  &self:SystemSpawnScript.SystemScriptSource.SystemScriptGraph.NiagaraNodeAssignment_2.SetVariables_#3
  &self:SystemSpawnScript.SystemScriptSource.SystemScriptGraph.NiagaraNodeAssignment_3.SetVariables_#4
  &self:SystemSpawnScript.SystemScriptSource.SystemScriptGraph.NiagaraNodeAssignment_4.SetVariables_#5
  &SystemState
 ]
}
emitters {
 BodyGlows {
  emitterUpdateScript [
   &EmitterState()
   &SpawnBurst_Instantaneous(Spawn Count=11, Use Spawn Probability=false, UseLoopCountLimit=true)
  ]
  particleSpawnScript [
   &InitializeParticle(
    Color=<rgba 1 0.3595 0.0468 0.25>
    Mesh Scale Mode=::NewEnumerator3
    Mesh Uniform Scale=10
    Sprite Size Mode=::NewEnumerator3
    Uniform Sprite Size=&FN_FrontEndFloatModifier(Gameplay=40, Front End=35)
    UsePositionOffset=false
    Write Material Random=true
   )
   &PartitionParticles(Number Of Partitions=11)
   &DynamicMaterialParameters(
    Index 0 Param 1=3
    Index 0 Param 2=10
    Index 0 Param 3=0.5
    Index 0 Param 4=1
    Module.Index0Float0Write=true
    Module.Index0Float1Write=true
    Module.Index0Float2Write=true
    Module.Index0Float3Write=true
    Module.Index1Float0Write=true
    Module.Index1Float1Write=true
    Module.Index1Float2Write=true
    Module.Index1Float3Write=true
    Write Parameter Index 1=::NewEnumerator1
   )
   /SetVariables(Particles.SpriteSize=&Multiply_Vector2D(
     A=$Particles.SpriteSize
     B=&SelectVector2DFromArray(
      Direct Array Index=$Particles.PartitionParticles.Partition
      Vector 2D Selection Array=DataInterface() {propertyValues {FloatData [<2 2>, <2 2>, <2 2>, <1 1>, <1 1>, <2 2>, <3 3>, <2 3>, <2 3>, <1.5 2.5>, <1.5 2.5>]}}
     )
    )) {isSetParameters=true}
  ]
  particleUpdateScript [
   &ParticleState(Kill Particles When Lifetime Has Elapsed=false, Let Infinitely Lived Particles Die When Emitter Deactivates=true)
   &SkeletalMeshLocation(
    Bone / Socket Index=$Particles.PartitionParticles.Partition
    Bone Sampling Mode=::NewEnumerator4
    MaskSpawnGroup=false
    Skeletal Mesh=$System.SkelMesh
   )
   &Color(Color=&Multiply_LinearColorByFloat(
     Linear Color=&LerpLinearColors(StartColor=<rgba 1 0.5067 0 0.2>, EndColor=<rgba 4 1.032 0.0037 0.35>, LerpFactor=$User.ParaGlideLeanAlpha)
     Float=&FN_FrontEndFloatModifier(Gameplay=1, Front End=&Lerp_Float(A=1, B=0.5, Alpha=$User.ParaGlideLeanAlpha))
    ))
  ]
  rendererClasses [&NiagaraSpriteRendererProperties, &NiagaraMeshRendererProperties]
  renderers [
   {rendererClass=&NiagaraSpriteRendererProperties}
   {rendererClass=&NiagaraMeshRendererProperties, rendererIndex=1}
  ]
 }
 Debris {
  emitterUpdateScript [
   &EmitterState()
   &SpawnRate(
    Spawn Probability=&Lerp_Float(A=0.25, B=1, Alpha=$User.ParaGlideLeanAlpha)
    SpawnRate=5
    Use Spawn Probability=true
   )
  ]
  particleSpawnScript [
   &InitializeParticle(
    Color=<rgba 1 0.4163 0.0695 1>
    Lifetime=&RandomRangeFloat(Minimum=1.5, Maximum=2)
    Mesh Scale Mode=::NewEnumerator3
    Mesh Uniform Scale=&RandomRangeFloat(Minimum=0.5, Maximum=1.5)
    Sprite Size Max=<30 30>
    Sprite Size Min=<20 20>
    Sprite Size Mode=::NewEnumerator2
    UsePositionOffset=false
    Write Material Random=true
   )
   &ShapeLocation(
    MaskSpawnGroup=false
    Non Uniform Scale=<1 1 -0.5>
    Offset=&VectorFromCurve(
     VectorCurve=DataInterface() {propertyValues {XCurve {keys [<at 0 0>,<at 1 0>]}, YCurve {keys [<at 0 0>,<at 1 0>]}, ZCurve {keys [<at 0 -200>,<at 1 -500>]}}}
     CurveIndex=$User.ParaGlideLeanAlpha
    )
    Offset Mode=::NewEnumerator0
    Override Seed=false
    Ring Radius=200
    Shape Primitive=::NewEnumerator4
    UseOwnerScale=true
   )
   &AddVelocity(Velocity=<0 0 1>, Velocity Speed Scale=&RandomRangeFloat(Minimum=100, Maximum=200))
   &InitialMeshOrientation(Use Rotation Vector=false)
   /SetVariables(
    Particles.DebrisDyn01:Vector4f=&MakeVector4(X=&RandomRangeFloat(Minimum=0.5, Maximum=2), Y=0, Z=0, W=0)
    Particles.MeshIndex=&RandomRangeInt(Minimum=0, Maximum=2)
   ) {isSetParameters=true}
   &DynamicMaterialParameters(
    Index 0 Param 1=3
    Index 0 Param 3=0.5
    Index 0 Param 4=1
    Module.Index0Float0Write=true
    Module.Index0Float1Write=true
    Module.Index0Float2Write=true
    Module.Index0Float3Write=true
    Module.Index1Float0Write=true
    Module.Index1Float1Write=true
    Module.Index1Float2Write=true
    Module.Index1Float3Write=true
    Write Parameter Index 1=::NewEnumerator1
   )
  ]
  particleUpdateScript [
   &ParticleState()
   &ScaleSpriteSize(Uniform Curve Sprite Scale=DataInterface() {propertyValues {Curve {keys [
        {interpMode=RCIM_Cubic, tangentMode=RCTM_User, time=0, value=0}
        <at 0.25 1>
        {interpMode=RCIM_Cubic, tangentMode=RCTM_User, time=0.8349, value=1}
        <at 1 0>
       ]}}})
   &ScaleMeshSize(Scale Factor=&VectorFromFloat(Value=&FloatFromCurve(
      FloatCurve=DataInterface() {propertyValues {Curve {keys [
          {interpMode=RCIM_Cubic, tangentMode=RCTM_User, time=0, value=0}
          <at 0.25 1>
          {interpMode=RCIM_Cubic, tangentMode=RCTM_User, time=0.8349, value=1}
          <at 1 0>
         ]}}}
      Scale Curve=&FN_FrontEndFloatModifier(
       Gameplay=/Game/Effects/Niagara/Modules/Math/FN_CameraDistance(
        0Dist:NiagaraFloat=50
        1Dist:NiagaraFloat=150
        Camera Query:NiagaraDataInterfaceCamera=DataInterface()
        Position:NiagaraPosition=$Particles.Position
       )
       Front End=1
      )
     )))
   &UpdateMeshOrientation(Rotation Vector=&RandomRangeVector(Minimum=<-1 -1 -1>, Maximum=<1 1 1>))
   &ScaleColor(
    Scale Alpha=&FloatFromCurve(FloatCurve=DataInterface() {propertyValues {Curve {keys [{interpMode=RCIM_Cubic, tangentMode=RCTM_User, time=0, value=1}, <at 1 0>]}}})
    Scale RGB=&Multiply_VectorByFloat(
     Vector=<1 1 1>
     Float=&FloatFromCurve(FloatCurve=DataInterface() {propertyValues {Curve {keys [<at 0 1>, {interpMode=RCIM_Cubic, tangentMode=RCTM_User, time=0.0492, value=3}, <at 0.25 1>, <at 1 1>]}}})
    )
    ScaleA=true
    ScaleRGB=true
   )
   &GravityForce(Gravity=&Lerp_Vector(
     A=&RandomRangeVector(Minimum=<0 0 200>, Maximum=<0 0 500>)
     B=&RandomRangeVector(Minimum=<0 0 800>, Maximum=<0 0 1500>)
     Alpha=$User.ParaGlideLeanAlpha
    ))
   &ScaleVelocity(Velocity Scale=&MakeVector(
     X=0
     Y=0
     Z=&FloatFromCurve(FloatCurve=DataInterface() {propertyValues {Curve {keys [<at 0 1>,<at 1 3>]}}}, CurveIndex=$User.ParaGlideLeanAlpha)
    )) {enabled=false}
   &SolveForcesAndVelocity(Clamp Velocity=false, Limit Acceleration=false)
   &CameraOffset(Camera Offset Amount=&FN_FrontEndFloatModifier(Gameplay=0, Front End=-100))
  ]
  rendererClasses [&NiagaraMeshRendererProperties, &NiagaraSpriteRendererProperties]
  renderers [
   {rendererClass=&NiagaraMeshRendererProperties, rendererIndex=0}
   {rendererClass=&NiagaraSpriteRendererProperties, rendererIndex=1}
  ]
 }
 HeatShield {
  emitterUpdateScript [
   &EmitterState()
   &SpawnBurst_Instantaneous(Use Spawn Probability=false, UseLoopCountLimit=true)
  ]
  particleSpawnScript [
   &InitializeParticle(
    Mesh Scale Mode=::NewEnumerator3
    Mesh Uniform Scale=5
    Position Offset=<0 0 100>
    Sprite Size Mode=::NewEnumerator3
    UsePositionOffset=true
    Write Material Random=true
   )
   &InitialMeshOrientation(Mesh Orientation Mode=::NewEnumerator2, Use Rotation Vector=true)
  ]
  particleUpdateScript [
   &ParticleState(Kill Particles When Lifetime Has Elapsed=false, Let Infinitely Lived Particles Die When Emitter Deactivates=true)
   &ScaleMeshSize(Scale Factor=&VectorFromCurve(
     VectorCurve=DataInterface() {propertyValues {XCurve {keys [<at 0 0.5>,<at 1 1>]}, YCurve {keys [<at 0 0.5>,<at 1 1>]}, ZCurve {keys [<at 0 0.25>,<at 1 1>]}}}
     CurveIndex=$Emitter.Age
    )) {module=ScaleMeshSize001}
   &ScaleMeshSize(Scale Factor=&Lerp_Vector(
     A=&Lerp_Vector(
      A=<1.5 1.5 1>
      B=<1.25 1.25 1>
      Alpha=&Abs_Float(Float:NiagaraFloat=&DotProduct(Vector A=<0 0 1>, Vector B=&ReturnCameraVector(Camera Vectors=::NewEnumerator1, Coordinate Space=::World)))
     )
     B=<1 1 1>
     Alpha=$User.ParaGlideLeanAlpha
    ))
   /SetVariables(Particles.Position=&Lerp_Vector(
     A=<0 0 100>
     B=&FN_FrontEndVector3Modifier(Gameplay=<0 0 0>, Front End=<0 40 80>)
     Alpha=$User.ParaGlideLeanAlpha
    )) {isSetParameters=true}
   &Color(Color=&LerpLinearColors(
     StartColor=<rgba 1 0.9287 0.8646 0.9>
     EndColor=&ColorFromCurve(
      ColorCurve=DataInterface() {propertyValues {
        AlphaCurve {keys [<at 0 0.5 RCIM_Cubic>,<at 1 0.5 RCIM_Cubic>]}
        BlueCurve {keys [<at 0 0.035 RCIM_Cubic>,<at 1 0 RCIM_Cubic>]}
        GreenCurve {keys [<at 0 0.8793 RCIM_Cubic>,<at 1 0.4998 RCIM_Cubic>]}
        RedCurve {keys [<at 0 1 RCIM_Cubic>,<at 1 0.75 RCIM_Cubic>]}
       }}
      CurveIndex=&FNTimeOfDayScalar(Night Scale=1, Morning Scale=0, Day Scale=0, Evening Scale=1)
     )
     LerpFactor=&FloatFromCurve(
      FloatCurve=DataInterface() {propertyValues {Curve {keys [{interpMode=RCIM_Cubic, tangentMode=RCTM_User, time=0.6, value=0}, <at 1 1>]}}}
      CurveIndex=&Abs_Float(Float:NiagaraFloat=&DotProduct(Vector A=<0 0 1>, Vector B=&ReturnCameraVector(Camera Vectors=::NewEnumerator1, Coordinate Space=::World)))
     )
    ))
  ]
  rendererClasses [&NiagaraMeshRendererProperties]
  renderers [
   {rendererClass=&NiagaraMeshRendererProperties, rendererIndex=0}
   {rendererClass=&NiagaraMeshRendererProperties, rendererIndex=1}
  ]
 }
 Lightning {
  emitterUpdateScript [
   &EmitterState(
    Life Cycle Mode=::NewEnumerator1
    Loop Duration=&RandomRangeFloat(Minimum=0.25, Maximum=3.5)
    Recalculate Duration Each Loop=true
    UseLoopDelay=false
   )
   &SpawnRate(Spawn Probability=&Lerp_Float(A=0, B=1, Alpha=$User.ParaGlideLeanAlpha), SpawnRate=10, Use Spawn Probability=true)
  ]
  particleSpawnScript [
   &PartitionParticles()
   &InitializeParticle(
    Color=&RandomRangeLinearColor(Minimum=<rgba 20 4.3087 0 1>, Maximum=<rgba 10 0.6222 0 1>)
    Lifetime=&RandomRangeFloat(Minimum=0.5, Maximum=1)
    Mesh Scale=&Multiply_VectorByFloat(Vector=&RandomRangeVector(Minimum=<0.2 0.2 1.5>, Maximum=<1 1 5>), Float=10)
    Mesh Scale Mode=::NewEnumerator4
    Sprite Size Mode=::NewEnumerator3
    UsePositionOffset=false
    Write Material Random=true
   )
   &InitialMeshOrientation(
    Mesh Orientation Mode=::NewEnumerator2
    Rotation=&RandomRangeVector(Minimum=<0 0 0>, Maximum=<0 0 1>)
    Use Rotation Vector=true
   )
   &CameraOffset(Camera Offset Amount=5)
   &ShapeLocation(
    MaskSpawnGroup=false
    Non Uniform Scale=<1 1 -0.5>
    Offset=&RandomRangeVector(Minimum=<0 0 -50>, Maximum=<0 0 -75>)
    Offset Mode=::NewEnumerator0
    Override Seed=false
    Ring Radius=50
    Shape Primitive=::NewEnumerator4
    UseOwnerScale=true
   )
  ]
  particleUpdateScript [
   &DynamicMaterialParameters(
    Index 0 Param 1=&FloatFromCurve(
     FloatCurve=DataInterface() {propertyValues {Curve {keys [
         {arriveTangent=3.0269, interpMode=RCIM_Cubic, leaveTangent=3.0269, tangentMode=RCTM_User, time=0, value=0}
         {tangentMode=RCTM_User, time=1, value=1}
        ]}}}
     Scale Curve=&RandomRangeFloat(Minimum=0, Maximum=1)
    )
    Index 0 Param 2=0.7
    Index 0 Param 3=&FloatFromCurve(FloatCurve=DataInterface() {propertyValues {Curve {keys [<at 0.25 0 RCIM_Cubic>, {arriveTangent=5.448, interpMode=RCIM_Cubic, leaveTangent=5.448, tangentMode=RCTM_User, time=1, value=3}]}}})
    Index 0 Param 4=&FloatFromCurve(
     FloatCurve=DataInterface() {propertyValues {Curve {keys [
         {arriveTangent=2.5109, interpMode=RCIM_Cubic, leaveTangent=2.5109, tangentMode=RCTM_User, time=0, value=0}
         {interpMode=RCIM_Cubic, tangentMode=RCTM_User, time=1, value=1}
        ]}}}
     Scale Curve=&RandomRangeFloat(Minimum=0, Maximum=3)
    )
    Index 1 Param 1=&FloatFromCurve(
     FloatCurve=DataInterface() {propertyValues {Curve {keys [<at 0.25 1>,<at 1 0>]}}}
     Scale Curve=&RandomRangeFloat(Minimum=1, Maximum=3)
    )
    Index 1 Param 2=&RandomRangeFloat(Minimum=0, Maximum=1)
    Index 1 Param 3=&RandomRangeFloat(Minimum=1, Maximum=3)
    Index 1 Param 4=1
    Index 2 Param 1=&FloatFromCurve(FloatCurve=DataInterface() {propertyValues {Curve {keys [{arriveTangent=-2.6239, interpMode=RCIM_Cubic, leaveTangent=-2.6239, tangentMode=RCTM_User, time=0, value=1}, <at 1 0.3>]}}})
    Module.Param0WriteEnabled=true
    Module.Param0WriteEnabled1=true
    Module.Param0WriteEnabled2=true
    Module.Param1WriteEnabled=true
    Module.Param1WriteEnabled1=true
    Module.Param1WriteEnabled2=true
    Module.Param2WriteEnabled=true
    Module.Param2WriteEnabled1=true
    Module.Param2WriteEnabled2=true
    Module.Param3WriteEnabled=true
    Module.Param3WriteEnabled1=true
    Module.Param3WriteEnabled2=true
    Write Parameter Index 1=true
    Write Parameter Index 2=true
   )
   &ParticleState()
   &ScaleMeshSize(Scale Factor=&VectorFromCurve(VectorCurve=DataInterface() {propertyValues {XCurve {keys [<at 0 1>,<at 1 1>]}, YCurve {keys [<at 0 1>,<at 1 1>]}, ZCurve {keys [<at 0 0.2>,<at 1 2>]}}}))
   &ScaleColor(
    Scale Alpha=&Multiply_Float(
     A=1
     B=&FloatFromCurve(FloatCurve=DataInterface() {propertyValues {Curve {keys [<at 0 0>,<at 0.1 0>,<at 0.2 0.8>,<at 1 0>]}}})
    )
    Scale RGB=&Multiply_VectorByFloat(
     Vector=<1 1 1>
     Float=&FloatFromCurve(FloatCurve=DataInterface() {propertyValues {Curve {keys [
          {arriveTangent=4.9531, interpMode=RCIM_Cubic, leaveTangent=4.9531, tangentMode=RCTM_User, time=0, value=1}
          {arriveTangent=-2.0019, interpMode=RCIM_Cubic, leaveTangent=-2.0019, time=0.05, value=10}
          {arriveTangent=-0.2096, interpMode=RCIM_Cubic, leaveTangent=-0.2096, tangentMode=RCTM_User, time=0.1983, value=0.6029}
          <at 1 0>
         ]}}})
    )
    ScaleA=true
    ScaleRGB=true
   )
   &SolveForcesAndVelocity(Clamp Velocity=false, Limit Acceleration=false)
  ]
  rendererClasses [&NiagaraMeshRendererProperties]
  renderers [{rendererClass=&NiagaraMeshRendererProperties, rendererIndex=0}]
 }
 Runes {
  emitterUpdateScript [
   &EmitterState()
   &SpawnRate(
    Spawn Probability=&Lerp_Float(A=0.25, B=1, Alpha=$User.ParaGlideLeanAlpha)
    SpawnRate=25
    Use Spawn Probability=true
   )
  ]
  particleSpawnScript [
   &InitializeParticle(
    Lifetime=&RandomRangeFloat(Minimum=0.5, Maximum=2)
    Sprite Size Mode=::NewEnumerator1
    Uniform Sprite Size Max=12
    Uniform Sprite Size Min=6
    UsePositionOffset=false
    Write Material Random=true
   )
   &ShapeLocation(
    Hemisphere Distribution=<1 0.5>
    MaskSpawnGroup=false
    Non Uniform Scale=<1 1 -0.5>
    Offset=<0 0 -25>
    Offset Mode=::NewEnumerator0
    Override Seed=false
    Sphere Radius=125
    UseOwnerScale=true
   )
   &AddVelocity(Velocity=<0 0 1>, Velocity Speed Scale=&RandomRangeFloat(Minimum=0, Maximum=100))
  ]
  particleUpdateScript [
   &ParticleState()
   &DynamicMaterialParameters(
    Index 0 Param 1=&Waveform(
     "Waveform [1]"=::NewEnumerator9
     "[1] Phase Scale"=&RandomRangeFloat(Minimum=0, Maximum=1)
     "[1] Frequency"=&RandomRangeFloat(Minimum=6, Maximum=10)
     "[1] Amplitude Min/Max"=<0 4>
    )
    Index 0 Param 2=1
    Index 0 Param 3=1
    Module.Index0Float0Write=true
    Module.Index0Float1Write=true
    Module.Index0Float2Write=true
    Module.Index0Float3Write=true
   )
   &SubUVAnimation(
    Random Start Frame=true
    "Sprite Renderer "=DataInterface() {propertyValues {SpriteRenderer=&self:Runes_0.NiagaraSpriteRendererProperties_0}}
    SubUV Lookup Index:NiagaraFloat=&Add_Float(
     A=&Multiply_Float(A=$Engine.System.Age, B=&RandomRangeFloat(Minimum=0.05, Maximum=0.2))
     B=&RandomRangeFloat(Minimum=0, Maximum=1)
    )
    SubUV Lookup Index Scale:NiagaraFloat=&FloatFromCurve(
     FloatCurve=DataInterface() {propertyValues {Curve {keys [<at 0 4>,<at 0.1 1>,<at 1 1>]}}}
     Scale Curve=&RandomRangeFloat(Minimum=1, Maximum=2)
    )
    UseEndFrame=false
    UseStartFrame=false
   ) {module=SubUVAnimation001}
   &ScaleColor(
    Scale Alpha=&FloatFromCurve(FloatCurve=DataInterface() {propertyValues {Curve {keys [{interpMode=RCIM_Cubic, tangentMode=RCTM_User, time=0, value=1}, <at 1 0>]}}})
    Scale RGB=&Multiply_VectorByFloat(
     Vector=<1 1 1>
     Float=&FloatFromCurve(FloatCurve=DataInterface() {propertyValues {Curve {keys [<at 0 1>, {interpMode=RCIM_Cubic, tangentMode=RCTM_User, time=0.0492, value=3}, <at 0.25 1>, <at 1 1>]}}})
    )
    ScaleA=true
    ScaleRGB=true
   )
   &GravityForce(Gravity=&Lerp_Vector(
     A=&RandomRangeVector(Minimum=<0 0 200>, Maximum=<0 0 500>)
     B=&RandomRangeVector(Minimum=<0 0 400>, Maximum=<0 0 700>)
     Alpha=$User.ParaGlideLeanAlpha
    ))
   &SolveForcesAndVelocity(Clamp Velocity=false, Limit Acceleration=false)
  ]
  rendererClasses [&NiagaraSpriteRendererProperties]
  renderers [{rendererClass=&NiagaraSpriteRendererProperties}]
 }
 Sparks {
  emitterUpdateScript [
   &EmitterState()
   &SpawnRate(SpawnRate=75, Use Spawn Probability=false)
  ]
  particleSpawnScript [
   &InitializeParticle(
    Color=<rgba 10 3.9485 0.4292 1>
    Lifetime=&RandomRangeFloat(Minimum=0.5, Maximum=1)
    Sprite Size Max=<0.5 15>
    Sprite Size Min=<0.5 5>
    Sprite Size Mode=::NewEnumerator2
    UsePositionOffset=false
    Write Material Random=true
   )
   &ShapeLocation(
    Disc Coverage=1
    MaskSpawnGroup=false
    Offset=<0 0 -15>
    Offset Mode=::NewEnumerator0
    Override Seed=false
    Ring Radius=125
    Shape Primitive=::NewEnumerator4
    UseOwnerScale=true
   )
   &AddVelocity(Velocity=<0 0 10>)
  ]
  particleUpdateScript [
   &ParticleState()
   &GravityForce(Gravity=&Multiply_Vector(
     A=&RandomRangeVector(Minimum=<-100 -100 300>, Maximum=<100 100 500>)
     B=&Lerp_Vector(A=<1 1 1>, B=<1 1 5>, Alpha=$User.ParaGlideLeanAlpha)
    ))
   &CurlNoiseForce(Bias Noise=false, Noise Strength=500, Pan Noise=false)
   &SolveForcesAndVelocity(Clamp Velocity=false, Limit Acceleration=false)
   &ScaleColor(
    Scale Alpha=&FloatFromCurve(FloatCurve=DataInterface() {propertyValues {Curve {keys [{interpMode=RCIM_Cubic, tangentMode=RCTM_User, time=0, value=0}, <at 0.0925 1>, <at 1 0>]}}})
    ScaleA=true
    ScaleRGB=true
   )
  ]
  rendererClasses [&NiagaraSpriteRendererProperties]
  renderers [{rendererClass=&NiagaraSpriteRendererProperties}]
 }
}
summary {
 systemName=NS_Contrail_SpacePlunge
 emitters [
  {emitterName=HeatShield, rendererClasses [&NiagaraMeshRendererProperties]}
  {emitterName=Runes, rendererClasses [&NiagaraSpriteRendererProperties]}
  {emitterName=Lightning, rendererClasses [&NiagaraMeshRendererProperties]}
  {emitterName=Sparks, rendererClasses [&NiagaraSpriteRendererProperties]}
  {emitterName=BodyGlows, rendererClasses [&NiagaraSpriteRendererProperties, &NiagaraMeshRendererProperties]}
  {emitterName=Debris, rendererClasses [&NiagaraMeshRendererProperties, &NiagaraSpriteRendererProperties]}
 ]
 userVariables [
  User.bIsFrontEnd {type=NiagaraBool}
  User.bIsFrontEndPreview {type=NiagaraBool}
  User.ParaGlideLeanAlpha {type=NiagaraFloat, defaultValue=0, description=""}
  User.RibbonIDOffset {type=NiagaraInt32}
 ]
}
